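# Physics asset for the Jumpstart pickaxe skeletal mesh, defining collision and joints for its two dangling cables.
# A single box body covers the two-handed handle and anchors two four-segment capsule chains, cable A and cable B.
# Each segment is constrained to the one before it with wide swing limits and tight twist, so the cables can bend but barely twist.
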PhysicsAsset /BRCosmetics/Weapons/FORT_Melee/Pickaxe_Jumpstart/Meshes/SK_Pickaxe_Jumpstart_Physics
bodies {
 Melee_TwoHanded_Handle [{shapeType=Box, extentX=7.5, extentY=47.0998, extentZ=20.9405, center=<0 6.431 75.3085>}]
 dyn_cable_a_01 [{shapeType=Capsule, length=1.8, radius=1.5, center=<1.5 0 0>, rotation=<rot 90 0 0>}]
 dyn_cable_a_02 [{shapeType=Capsule, length=5, radius=1.5, center=<2.7 0 0>, rotation=<rot 90 0 0>}]
 dyn_cable_a_03 [{shapeType=Capsule, length=7, radius=1.5, center=<3.5 0 0>, rotation=<rot 90 0 0>}]
 dyn_cable_a_04 [{shapeType=Capsule, length=8, radius=1.5, center=<4.5 0 0>, rotation=<rot 90 0 0>}]
 dyn_cable_b_01 [{shapeType=Capsule, length=1.8, radius=1.5, center=<1.5 0 0>, rotation=<rot 90 0 0>}]
 dyn_cable_b_02 [{shapeType=Capsule, length=5, radius=1.5, center=<3.3 0 0>, rotation=<rot 90 0 0>}]
 dyn_cable_b_03 [{shapeType=Capsule, length=8, radius=1.5, center=<3.8 0 0>, rotation=<rot 90 0 0>}]
 dyn_cable_b_04 [{shapeType=Capsule, length=9, radius=1.5, center=<4.5 0 0>, rotation=<rot 90 0 0>}]
}
constraints [
 {bone1Name=dyn_cable_a_01, bone2Name=Melee_TwoHanded_Handle, swing1LimitDegrees=60, swing2LimitDegrees=45, twistLimitDegrees=5}
 {bone1Name=dyn_cable_a_02, bone2Name=dyn_cable_a_01, swing1LimitDegrees=60, swing2LimitDegrees=45, twistLimitDegrees=5}
 {bone1Name=dyn_cable_a_03, bone2Name=dyn_cable_a_02, swing1LimitDegrees=60, swing2LimitDegrees=45, twistLimitDegrees=5}
 {bone1Name=dyn_cable_a_04, bone2Name=dyn_cable_a_03, swing1LimitDegrees=60, swing2LimitDegrees=45, twistLimitDegrees=5}
 {bone1Name=dyn_cable_b_01, bone2Name=Melee_TwoHanded_Handle, swing1LimitDegrees=60, swing2LimitDegrees=45, twistLimitDegrees=5}
 {bone1Name=dyn_cable_b_02, bone2Name=dyn_cable_b_01, swing1LimitDegrees=60, swing2LimitDegrees=45, twistLimitDegrees=5}
 {bone1Name=dyn_cable_b_03, bone2Name=dyn_cable_b_02, swing1LimitDegrees=60, swing2LimitDegrees=45, twistLimitDegrees=5}
 {bone1Name=dyn_cable_b_04, bone2Name=dyn_cable_b_03, swing1LimitDegrees=60, swing2LimitDegrees=45, twistLimitDegrees=5}
]
